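# Looping interior ambience cue for a metal shipping container prop, on the indoor ambient sound class.
# A quality switch selects between two three-loop mixes sharing industrial and large-room tones, with quad container audio in the first two slots and stereo in the last two.
# Two nested endlessly looping envelopers add slow pitch modulation and gentle volume swell.
# Box attenuation with short falloff confines the sound, and concurrency prevents new instances over the limit.
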SoundCue /Game/Sounds/Asteria/Environment_Actors/MetalShippingContainer/Ambience_Metal_ShippingContainer_Prop_Loop_Cue
root=Enveloper_2
nodes [
 {
  id=Enveloper_1
  children [QualityLevel_0]
  properties {
   bLoop=true
   bLoopIndefinitely=true
   loopEnd=30
   pitchCurve {editorCurveData {keys [{interpMode=RCIM_Cubic, time=0}, <at 15.9468 0.9181 RCIM_Cubic>, {interpMode=RCIM_Cubic, time=30}]}}
   volumeCurve {editorCurveData {keys [<at 0 1>]}}
  }
 }
 {
  id=Enveloper_2
  children [Enveloper_1]
  properties {
   bLoop=true
   bLoopIndefinitely=true
   loopEnd=20
   pitchCurve {editorCurveData {keys [{time=0}]}}
   volumeCurve {editorCurveData {keys [
      {arriveTangent=0, interpMode=RCIM_Cubic, time=0, value=1}
      {arriveTangent=0, interpMode=RCIM_Cubic, leaveTangent=0, time=10, value=1.05}
      {interpMode=RCIM_Cubic, leaveTangent=0, time=20, value=0.9992}
     ]}}
  }
 }
 {id=Mixer_0, children [WavePlayer_0,WavePlayer_3,WavePlayer_1], properties {inputVolume [0.5,0.3,0.3]}}
 {id=Mixer_2, children [WavePlayer_3,WavePlayer_1,WavePlayer_2], properties {inputVolume [0.5,0.3,0.3]}}
 {id=QualityLevel_0, children [Mixer_0,Mixer_0,Mixer_2,Mixer_2]}
 {
  id=WavePlayer_0
  properties {bLooping=true, soundWaveAssetPtr=&/AMB_Shipping_Container_Loop_Int_Quad}
 }
 {
  id=WavePlayer_1
  properties {bLooping=true, soundWaveAssetPtr=/Game/Sounds/Fort_Starlight/S4/SoundWaves/AMBI_Starlight_S4_HauntedHotel_LargeRoom_Int_Night_Loop}
 }
 {
  id=WavePlayer_2
  properties {bLooping=true, soundWaveAssetPtr=&/AMB_Shipping_Container_Loop_Int_Stereo}
 }
 {id=WavePlayer_3, properties {bLooping=true, soundWaveAssetPtr=/Game/Sounds/Trans_AmbientLoops/RoomTones/IndustrialSmall002_Loop_Stereo}}
]
properties {
 bOverrideAttenuation=true
 bOverrideConcurrency=true
 maxDistance=261.5887
 soundClassObject=&Fort_Ambient_Indoors_Athena
 volumeMultiplier=0.8
 attenuationOverrides {
  attenuationShape=Box
  bEnableReverbSend=false
  falloffDistance=5
  nonSpatializedRadiusEnd=500
  nonSpatializedRadiusStart=250
  attenuationShapeExtents=<256.5887 111.3225 136>
 }
 concurrencyOverrides {resolutionRule=PreventNew}
}
