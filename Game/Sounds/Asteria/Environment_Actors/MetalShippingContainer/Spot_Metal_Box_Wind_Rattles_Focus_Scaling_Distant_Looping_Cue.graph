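# Looping ambient SoundCue for the distant layer of wind rattling a metal shipping container, classed as indoor ambience.
# Mixes looped random gusty metal rattles, run through a Delay, with a random looping window-wind bed shaped by slow looping pitch and volume envelopes.
# The mix plays only on the first two quality levels.
# Box attenuation uses a custom band curve that is silent up close and far away, with interpolated listener focus scaling volume.
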SoundCue /Game/Sounds/Asteria/Environment_Actors/MetalShippingContainer/Spot_Metal_Box_Wind_Rattles_Focus_Scaling_Distant_Looping_Cue
root=QualityLevel_0
nodes [
 {id=Delay_0, children [Looping_0]}
 {
  id=Enveloper_0
  children [Random_1]
  properties {
   bLoop=true
   bLoopIndefinitely=true
   loopEnd=30
   pitchCurve {editorCurveData {keys [{interpMode=RCIM_Cubic, time=0}, <at 15 0.9 RCIM_Cubic>, {interpMode=RCIM_Cubic, time=30}]}}
   volumeCurve {editorCurveData {keys [<at 0 1>]}}
  }
 }
 {
  id=Enveloper_3
  children [Enveloper_0]
  properties {
   bLoop=true
   bLoopIndefinitely=true
   loopEnd=60
   pitchCurve {editorCurveData {keys [{time=0}]}}
   volumeCurve {editorCurveData {keys [
      {arriveTangent=-0.01, interpMode=RCIM_Cubic, time=0, value=1}
      <at 10 0.9 RCIM_Cubic>
      {arriveTangent=0.0047, interpMode=RCIM_Cubic, leaveTangent=0.0047, time=20, value=1}
      <at 31.1629 1>
      {arriveTangent=-0.0016, interpMode=RCIM_Cubic, leaveTangent=-0.0016, time=37.937, value=0.8}
      <at 46.7125 0.975>
      <at 60 1 RCIM_Cubic>
     ]}}
  }
 }
 {id=Looping_0, children [Random_0]}
 {id=Mixer_0, children [Delay_0,Enveloper_3], properties {inputVolume [0.6,0.5]}}
 {id=QualityLevel_0, children [Mixer_0,Mixer_0,null,null]}
 {id=Random_0, children [WavePlayer_0,WavePlayer_1,WavePlayer_4,WavePlayer_5,WavePlayer_6,WavePlayer_7], properties {weights [1,1,1,1,1,1]}}
 {id=Random_1, children [WavePlayer_3,WavePlayer_2,WavePlayer_8], properties {weights [1,1,1]}}
 {
  id=WavePlayer_0
  properties {soundWaveAssetPtr=&/Spot_Wind_Metal_Ratte_Gusty_Interior_01}
 }
 {
  id=WavePlayer_1
  properties {soundWaveAssetPtr=&/Spot_Wind_Metal_Ratte_Gusty_Interior_02}
 }
 {id=WavePlayer_2, properties {bLooping=true, soundWaveAssetPtr=/Game/Sounds/Helios/POI/MountOlympus/Spots/Spot_Olympus_Window_Wind_02}}
 {id=WavePlayer_3, properties {bLooping=true, soundWaveAssetPtr=/Game/Sounds/Helios/POI/MountOlympus/Spots/Spot_Olympus_Window_Wind_01}}
 {
  id=WavePlayer_4
  properties {soundWaveAssetPtr=&/Spot_Wind_Metal_Ratte_Gusty_Interior_03}
 }
 {
  id=WavePlayer_5
  properties {soundWaveAssetPtr=&/Spot_Wind_Metal_Ratte_Gusty_Interior_04}
 }
 {
  id=WavePlayer_6
  properties {soundWaveAssetPtr=&/Spot_Wind_Metal_Ratte_Gusty_Interior_05}
 }
 {
  id=WavePlayer_7
  properties {soundWaveAssetPtr=&/Spot_Wind_Metal_Ratte_Gusty_Interior_06}
 }
 {id=WavePlayer_8, properties {bLooping=true, soundWaveAssetPtr=/Game/Sounds/Helios/POI/MountOlympus/Spots/Spot_Olympus_Window_Wind_03}}
]
properties {
 bOverrideAttenuation=true
 maxDistance=475.9615
 soundClassObject=&Fort_Ambient_Indoors_Athena
 volumeMultiplier=0.5
 attenuationOverrides {
  attenuationShape=Box
  bEnableFocusInterpolation=true
  bEnableListenerFocus=true
  distanceAlgorithm=Custom
  falloffDistance=332.9337
  focusReleaseInterpSpeed=0.4
  nonFocusVolumeAttenuation=0.75
  attenuationShapeExtents=<140.3231 69.9932 100>
  customAttenuationCurve {editorCurveData {keys [<at 0.2 0>,<at 0.3 1>,<at 0.5 1>,<at 1.0081 0>]}}
 }
 preEffectBusSends [{}]
}
